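# Physics asset for the EnspierT2 vehicle body cosmetic, in its BR variant.
# It defines a single body on the root bone, with no collision shapes listed.
# It has no constraints, so no articulated or jointed simulation is set up between bones.
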PhysicsAsset /VehicleCosmetics/Bodies/EnspierT2/Meshes/Body_EnspierT2_Physics_BR
bodies {root []}
constraints []
